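# Lit DesertPoint set material that dresses Cosmos glider textures with dust and blends them into the Apollo landscape.
# Dust tint, metallic and roughness vary with downward-facing normals, height above the landscape and crystal-noise detail.
# Height-lerp blends the world-space result into landscape runtime virtual texture color, with a separate normal fade and vertex-normal-dependent blend distance.
# Material proxies skip the landscape blend; Forward and Mobile use a simplified texture-only path.
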Material /Game/Environments/Apollo/Sets/DesertPoint/Misk/M_DesertCosmos1
graph {
 properties {bTangentSpaceNormal=false, bUseMaterialAttributes=true, pixelDepthOffsetMode=PDOM_Legacy, usedShadingModels=Default Lit}
 nodes [
  n1 = VectorParameter() {parameterName=Dust, defaultValue=<rgba 0.3776 0.2526 0.1246 0>}
  n2 = TextureSampleParameter2D() {parameterName=Diffuse, texture=/Game/Gadgets/Assets/VinderTech_GliderChute/Glider_Cosmos_Male/Textures/T_Glider_Cosmos_Male_D}
  n3 = TextureSampleParameter2D() {
   parameterName=Normals
   samplerType=SAMPLERTYPE_Normal
   texture=/Game/Gadgets/Assets/VinderTech_GliderChute/Glider_Cosmos_Male/Textures/T_Glider_Cosmos_Male_N
  }
  n4 = Saturate(None=DotProduct(A=Transform(None=@n3.RGB), B=Constant3Vector() {constant=<rgba 0 0 -1 0>}))
  n5 = ScalarParameter() {defaultValue=255, parameterName=RVT Metal DustFade}
  n6 = Subtract(
   A=ComponentMask(None=WorldPosition().XYZ) {b=true}
   B=RuntimeVirtualTextureSampleParameter() {materialType=WorldHeight, parameterName=RVTH, virtualTexture=/Game/Athena/Apollo/Maps/Landscape/VT/RVT_Apollo_Landscape_Height}.WorldHeight
  )
  n7 = VertexNormalWS()
  n8 = Constant3Vector() {constant=<rgba 0 0 -1 0>}
  n9 = MaterialProxyReplace(Realtime=@n6, MaterialProxy=Saturate(None=DotProduct(A=@n7, B=@n8)))
  n10 = MaterialFunctionCall(
   A=@n1.RGB
   B=LinearInterpolate(A=@n2.RGB, B=@n1.RGB, Alpha=@n4)
   Transition Phase=Saturate(None=Divide(A=@n9, B=@n5))
   Height Texture=@n2.R
   Contrast=Constant()
  ) {materialFunction=&HeightLerp}
  n11 = TextureSampleParameter2D() {
   parameterName=SpecularMasks
   samplerType=SAMPLERTYPE_Masks
   texture=/Game/Gadgets/Assets/VinderTech_GliderChute/Glider_Cosmos_Male/Textures/T_Glider_Cosmos_Male_S
  }
  n12 = LinearInterpolate(A=@n11.G, Alpha=@n4) {constB=0.85}
  n13 = ScalarParameter() {defaultValue=0.85, parameterName=DustMet}
  n14 = ScalarParameter() {defaultValue=1.5, parameterName=Dust R}
  n15 = ScalarParameter() {parameterName=R1}
  n16 = ScalarParameter() {defaultValue=5, parameterName=R2}
  n17 = Divide(A=@n9, B=ScalarParameter() {defaultValue=200, parameterName=RVT R Fade})
  n18 = TextureSample(UVs=Multiply(A=TextureCoordinate(), B=ScalarParameter() {defaultValue=15, parameterName=DetailsScale})) {samplerType=SAMPLERTYPE_Masks, texture=/Game/Accessories/FORT_Backpacks/M_MED_Heist/Textures/T_CrystalNoise_M}
  n19 = LinearInterpolate(
   A=@n11.B
   B=LinearInterpolate(A=LinearInterpolate(A=@n15, B=@n16, Alpha=@n11.B), B=ScalarParameter() {defaultValue=0.35, parameterName=R3}, Alpha=Saturate(None=@n17))
   Alpha=@n18.R
  )
  n20 = SetMaterialAttributes!(
   MaterialAttributes
   Base Color=@n10.Results
   Metallic=LinearInterpolate(A=Multiply(A=@n12, B=@n13), B=@n12, Alpha=@n10.Alpha)
   Specular=@n11.R
   Roughness=LinearInterpolate(A=Multiply(A=@n14, B=@n19), B=@n19, Alpha=@n10.Alpha)
   Emissive Color
   Normal=@n3.RGB
  ) {attributeSetTypes [
    #1
    #2
    #3
    #4
    #5
    #6
   ]}
  n21 = MaterialFunctionCall!(Material=@n20) {materialFunction=/Game/Athena/Apollo/Maps/Landscape/VT/MaterialFunctions/TransformTangentToWorld}
  n22 = RuntimeVirtualTextureSampleParameter!(UVs, World Position, Mip Level=Constant() {r=1}) {
   materialType=BaseColor_Normal_Specular_YCoCg
   mipValueMode=RVTMVM_MipLevel
   parameterName=RVT
   virtualTexture=/Game/Athena/Apollo/Maps/Landscape/VT/RVT_Apollo_Landscape
  }
  n23 = Saturate(None=ComponentMask(None=VertexNormalWS()) {b=true})
  n24 = LinearInterpolate(
   A=ScalarParameter() {defaultValue=15, parameterName=Blend RVT Sides}
   B=ScalarParameter() {defaultValue=155, parameterName="Blend RVT "}
   Alpha=@n23
  ) {constB=0}
  n25 = ScalarParameter() {defaultValue=-1, group=Textures, parameterName=AlphaMax}
  n26 = ScalarParameter() {defaultValue=2, group=Textures, parameterName=AlphaMin}
  n27 = Add(A=Multiply(A=@n18.R) {constB=0.35}, B=@n2.R)
  n28 = ScalarParameter() {defaultValue=1, parameterName=To RVT Contrast}
  n29 = BlendMaterialAttributes(
   A=SetMaterialAttributes!(MaterialAttributes, Base Color=@n22.BaseColor, Specular=@n22.BaseColor, Roughness=@n22.BaseColor) {attributeSetTypes [#1, #3, #4]}
   B=@n21.Result
   Alpha=MaterialFunctionCall(
    Transition Phase=Divide(A=@n6, B=@n24)
    Height Texture=LinearInterpolate(A=@n25, B=@n26, Alpha=@n27)
    Contrast=LinearInterpolate(A=@n28, Alpha=@n23) {constB=0}
   ) {materialFunction=&HeightLerp}.Alpha
  ) {vertexAttributeBlendType=UseA}
  n30 = SetMaterialAttributes!(MaterialAttributes=@n29, Normal=@n22.Normal) {attributeSetTypes [#6]}
  n31 = Saturate(None=Divide(A=@n6, B=ScalarParameter() {defaultValue=155, parameterName=Blend RVT Normals}))
  n32 = MaterialFunctionCall(
   NoRVT=@n21.Result
   RVT=MaterialProxyReplace(Realtime=BlendMaterialAttributes(A=@n30, B=@n29, Alpha=@n31) {vertexAttributeBlendType=UseA}, MaterialProxy=@n21.Result)
  ) {materialFunction=&FN_RVT_Switch_Material}
  n33 = SetMaterialAttributes!(MaterialAttributes, Base Color=@n2.RGB, Metallic=@n11.G, Specular=@n11.G, Roughness=@n11.G, Normal=Transform(None=@n3.RGB)) {attributeSetTypes [
    #1
    #2
    #3
    #4
    #6
   ]}
  ShadingPathSwitch(Default=@n32.Result, Deferred=@n32.Result, Forward=@n33, Mobile=@n33)
 ]
}
