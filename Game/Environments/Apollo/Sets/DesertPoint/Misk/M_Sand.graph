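# Environment mesh material for Apollo DesertPoint that layers sand over a world-aligned base surface and blends into the landscape.
# An optional dune layer height-lerps tinted sand by height relative to the actor, and an L2 layer is masked by UV1 or vertex normal.
# An optional RVT pass, skipped on mobile, blends landscape color and normals by height above terrain.
# Fortnite hit effects drive World Position Offset.
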Material /Game/Environments/Apollo/Sets/DesertPoint/Misk/M_Sand
graph {
 properties {
  bTangentSpaceNormal=false
  bUseMaterialAttributes=true
  bUsedWithNanite=true
  pixelDepthOffsetMode=PDOM_Legacy
  usedShadingModels=Default Lit
 }
 nodes [
  n1 = SetMaterialAttributes!(
   MaterialAttributes
   Base Color=TextureSampleParameter2D() {group=L2, parameterName=L2_Diffuse, texture=/Game/Packages/Fortress_Floors_Generic/Textures/T_FORT_Floors_PlasterCeiling01}.RGB
   Metallic
   Specular
   Roughness
   Normal=Transform(None=Constant3Vector() {constant=<rgba 0 0 1 0>})
  ) {attributeSetTypes [
    #1
    #2
    #3
    #4
    #5
   ]}
  n2 = TextureObjectParameter() {parameterName=SpecularMasks, samplerType=SAMPLERTYPE_Masks, texture=/Game/Environments/Sets/Arid/Textures/T_RooftopTarArid_01_S}
  n3 = MaterialFunctionCall!(
   Diffuse Texture Object=TextureObjectParameter() {parameterName=Diffuse, texture=/Game/Environments/Sets/Arid/Textures/T_RooftopTarArid_01}
   Spec Texture Object=@n2
   TextureSize=ScalarParameter() {defaultValue=512, parameterName=WSTileScale}
   --------------------------
   Use seperate settings for Z projection
   Use seperate texture for Z projection
   Texture Size Z projection
   Z Plane Diffuse Texture Object
   Z Plane Spec Texture Object
   "--------------------------      "
   ProjectionTransitionContrast
   Use Texture Alpha Heightmap to Refine the Transistion
   HeightLerpTransitionContrast
   PickTheDiffuseHeightChannel
   TheDiffuseHeightChannel
   "--------------------------       "
   NormalTextureObject=TextureObjectParameter() {parameterName=Normals, samplerType=SAMPLERTYPE_Normal, texture=/Game/Environments/Sets/Arid/Textures/T_RooftopTarArid_01_N}
   World Space Normals
   "--------------------------  "
   Export Float 4 Diffuse
   Export Float 4 Spec
   "--------------------------     "
   WorldPosition=Add(A=WorldPosition().XYZ, B=Constant3Vector() {constant=<rgba 256 256 0 0>})
   World Space Normal
  ) {materialFunction=/Engine/Functions/Engine_MaterialFunctions02/Texturing/WorldAlignedTextures_Complex}
  n4 = MaterialFunctionCall(Float3=@n3.Z Spec) {materialFunction=&BreakOutFloat3Components}
  n5 = SetMaterialAttributes(Base Color=@n3.Z Diffuse Texture, Metallic=@n4.G, Specular=@n4.G, Roughness=@n4.G, Normal=@n3.Z Diffuse Texture) {attributeSetTypes [
    #1
    #2
    #3
    #4
    #5
   ]}
  n6 = Multiply(A=TextureCoordinate(), B=ScalarParameter() {defaultValue=3, parameterName=Param_1})
  n7 = TextureSample(UVs=@n6) {texture=/Game/Environments/Apollo/Sets/Terrain/Textures/T_Apollo_Sand_D_01}
  n8 = Constant3Vector() {constant=<rgba 0.7188 0.4377 0.3594 0>}
  n9 = MaterialFunctionCall(In=ComponentMask(None=VertexNormalWS()) {b=true}, Contrast=Constant() {r=1}) {materialFunction=&CheapContrast}
  n10 = TextureSample(UVs=@n6) {samplerType=SAMPLERTYPE_Masks, texture=/Game/Environments/Apollo/Sets/Terrain/Textures/T_Apollo_Sand_S_01}
  n11 = Constant3Vector() {constant=<rgba 0 0 1 0>}
  n12 = TextureSample(UVs=@n6) {samplerType=SAMPLERTYPE_Normal, texture=/Game/Environments/Apollo/Sets/Terrain/Textures/T_Apollo_Sand_N_01}
  n13 = SetMaterialAttributes(
   Base Color=LinearInterpolate(A=Multiply(A=@n7.RGB, B=@n8), B=@n7.RGB, Alpha=@n9.Result)
   Metallic=@n10.G
   Specular=@n10.G
   Roughness=@n10.G
   Normal=Transform(None=LinearInterpolate(A=@n11, B=@n12.RGB, Alpha=@n9.Result))
  ) {attributeSetTypes [
    #1
    #2
    #3
    #4
    #5
   ]}
  n14 = Subtract(A=WorldPosition().XYZ, B=ActorPositionWS())
  n15 = Constant() {r=-0.9619}
  n16 = Divide(A=ComponentMask(None=Add(A=@n14, B=@n15)) {b=true}, B=Constant() {r=32})
  n17 = BlendMaterialAttributes(
   A=@n5
   B=@n13
   Alpha=MaterialFunctionCall(Transition Phase=Saturate(None=@n16), Height Texture=@n7.A, Contrast=Constant() {r=2}) {materialFunction=&HeightLerp}.Alpha
  )
  n18 = StaticSwitchParameter(True=ShadingPathSwitch(Default=@n17, Deferred=@n17, Forward=@n13, Mobile=@n13), False=@n5) {parameterName=DunesOnMesh}
  n19 = ComponentMask(None=VertexNormalWS()) {b=true}
  n20 = TextureCoordinate() {coordinateIndex=1}
  n21 = Saturate(None=Floor(None=ComponentMask(None=@n20) {r=true}))
  n22 = StaticSwitchParameter(
   True=BlendMaterialAttributes(A=@n1, B=@n18, Alpha=Saturate(None=@n19))
   False=StaticSwitchParameter(True=BlendMaterialAttributes(A=@n18, B=@n1, Alpha=@n21), False=@n18) {defaultValue=true, group=L2, parameterName=L2 True}
  ) {group=Vertex Normals, parameterName=VertexNormalsMask}
  n23 = RuntimeVirtualTextureSampleParameter() {materialType=BaseColor_Normal_Specular_YCoCg, parameterName=RVT, virtualTexture=/Game/Athena/Apollo/Maps/Landscape/VT/RVT_Apollo_Landscape}
  n24 = Subtract(
   A=ComponentMask(None=WorldPosition().XYZ) {b=true}
   B=RuntimeVirtualTextureSampleParameter() {materialType=WorldHeight, parameterName=RVT, virtualTexture=/Game/Athena/Apollo/Maps/Landscape/VT/RVT_Apollo_Landscape_Height}.WorldHeight
  )
  n25 = Saturate(None=ComponentMask(None=VertexNormalWS()) {b=true})
  n26 = LinearInterpolate(
   A=ScalarParameter() {defaultValue=10, parameterName=Blend RVT Sides}
   B=ScalarParameter() {defaultValue=45, parameterName="Blend RVT "}
   Alpha=@n25
  ) {constB=0}
  n27 = Divide(A=ComponentMask(None=WorldPosition().XYZ) {g=true, r=true}, B=ScalarParameter() {defaultValue=256, parameterName=DetailsScale})
  n28 = ScalarParameter() {defaultValue=0.5, parameterName=To RVT Contrast}
  n29 = BlendMaterialAttributes(
   A=SetMaterialAttributes!(MaterialAttributes, Base Color=@n23.BaseColor, Specular=@n23.BaseColor, Roughness=@n23.BaseColor) {attributeSetTypes [#1, #3, #4]}
   B=@n22
   Alpha=MaterialFunctionCall(
    Transition Phase=Divide(A=@n24, B=@n26)
    Height Texture=TextureSample(UVs=@n27) {samplerType=SAMPLERTYPE_Masks, texture=/Game/Accessories/FORT_Backpacks/M_MED_Heist/Textures/T_CrystalNoise_M}.R
    Contrast=LinearInterpolate(A=@n28, Alpha=@n25) {constB=0}
   ) {materialFunction=&HeightLerp}.Alpha
  ) {vertexAttributeBlendType=UseA}
  n30 = SetMaterialAttributes!(MaterialAttributes=@n29, Normal=@n23.Normal) {attributeSetTypes [#5]}
  n31 = ScalarParameter() {defaultValue=3, parameterName=Blend RVT Normals Sides}
  n32 = ScalarParameter() {defaultValue=4, parameterName=Blend RVT Normals}
  n33 = Saturate(None=Divide(A=@n24, B=LinearInterpolate(A=@n31, B=@n32, Alpha=@n25) {constB=0}))
  n34 = MaterialFunctionCall(
   NoRVT=@n22
   RVT=MaterialProxyReplace(Realtime=BlendMaterialAttributes(A=@n30, B=@n29, Alpha=@n33) {vertexAttributeBlendType=UseA}, MaterialProxy=@n22)
  ) {materialFunction=&FN_RVT_Switch_Material}
  SetMaterialAttributes!(
   MaterialAttributes=StaticSwitchParameter(True=ShadingPathSwitch(Default=@n34.Result, Deferred=@n34.Result, Forward=@n34.Result, Mobile=@n22), False=@n22) {parameterName=Use RVT}
   World Position Offset=MaterialFunctionCall!(
    WSDeformationCenter=VectorParameter() {bUseCustomPrimitiveData=true, group=z_code, parameterName=WSDeformationCenter, primitiveDataIndex=20}.RGB
    Code_DestructionAmount=ScalarParameter() {group=z_code, parameterName=Code_DestructionAmount}
    Code_UseUvForDamage
    Code_UseHitEffects=StaticBoolParameter() {defaultValue=true, group=z_code, parameterName=Code_UseHitEffects}
    Code_Emissive
    WSDamageRadius=ScalarParameter() {bUseCustomPrimitiveData=true, defaultValue=1, group=z_code, parameterName=WSDamageRadius, primitiveDataIndex=26}
    CrackDamageShader
    Diffuse
    ScalarModDamageCracks
    ScalarOffsetforDamageCracks
    ExposeLerpFunctionForInvertingTheHitDistanceCalc
    InvertDamageFalloff
    CrackWorldPosition
    WorldPosition
    DistanceCullFadeAnimationOffset=ScalarParameter() {bUseCustomPrimitiveData=true, group=z_code, parameterName=DistanceCullFadeAnimationOffset, primitiveDataIndex=12}
   ) {materialFunction=/Game/Packages/jl_tests/Fortnite_HitEffects}.Hit And Shrink WPO
  ) {attributeSetTypes [#6]}
 ]
}
